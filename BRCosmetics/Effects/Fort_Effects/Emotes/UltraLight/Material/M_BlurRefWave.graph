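# Translucent unlit refraction material for Niagara mesh particles in the UltraLight emote, producing a blurred wave distortion.
# A slope-sheared plasma texture is warped by noise A, while noise B blends two dynamic-parameter V offsets.
# The refined result, edge mask, particle alpha and camera-distance fade scale temporal dither that drives index of refraction, creating the blur.
# Forward and mobile paths use a constant 1, disabling the distortion.
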Material /BRCosmetics/Effects/Fort_Effects/Emotes/UltraLight/Material/M_BlurRefWave
graph {
 properties {
  bUsedWithNiagaraMeshParticles=true
  blendMode=BLEND_Translucent
  pixelDepthOffsetMode=PDOM_Legacy
  refractionMethod=RM_IndexOfRefraction
  shadingModel=MSM_Unlit
  twoSided=true
  usedShadingModels=Unlit
 }
 nodes [
  Constant3Vector() {constant=<rgba 0 -1 0 0>}
  n1 = ParticleColor()
  n2 = TextureObjectParameter() {
   group=---TextureA
   parameterName=TextureA_Texture
   samplerType=SAMPLERTYPE_LinearColor
   texture=/Game/Effects/Fort_Effects/Emotes/UltraLight/Texture/T_Plasma_A_C
  }
  n3 = Constant()
  n4 = DynamicParameter() {defaultValue=<rgba 0 0 1 1>}
  n5 = Add(A=ScalarParameter() {parameterName=V}, B=@n4.Param1)
  n6 = Add(A=ScalarParameter() {parameterName=V2}, B=@n4.Param2)
  n7 = TextureObjectParameter() {
   group=---NoiseB
   parameterName=NoiseB_Texture
   samplerType=SAMPLERTYPE_LinearGrayscale
   texture=/Game/Effects/Fort_Effects/Emotes/UltraLight/Texture/T_DirectionalNoise_Greyscale
  }
  n8 = VectorParameter() {
   group=---NoiseB
   parameterName=NoiseB_UVControl_Mark:********************
   channelNames {a=PanV, b=PanU, g=ScaleV, r=ScaleU}
   defaultValue=<rgba 1 1 0 0>
  }
  n9 = MaterialFunctionCall(Float2=TextureCoordinate()) {materialFunction=&BreakOutFloat2Components}
  n10 = ScalarParameter() {parameterName=Slope}
  n11 = Add(A=TextureCoordinate(), B=AppendVector(A=Multiply(A=@n9.G, B=@n10), B=Constant()))
  n12 = Add(
   A=Power(
    Base=MaterialFunctionCall(Texture=@n7, UVControl_RGB=@n8, UVControl_A=@n8, "[Override]UVInput"=@n11) {materialFunction=&FXCommon_TextureSample}.R
    Exp=ScalarParameter() {defaultValue=1, parameterName=NoiseB_P}
   )
   B=ScalarParameter() {parameterName=NoiseB_O}
  )
  n13 = VectorParameter() {
   group=---TextureA
   parameterName=TextureA_UVControl_Mark:********************
   channelNames {a=PanV, b=PanU, g=ScaleV, r=ScaleU}
   defaultValue=<rgba 1 1 0 0>
  }
  n14 = TextureObjectParameter() {group=---NoiseA, parameterName=NoiseA_Texture, texture=/Game/Characters/Player/Male/Medium/Bodies/M_MED_Viper/Textures/Glitch/T_SoftNoise}
  n15 = VectorParameter() {
   group=---NoiseA
   parameterName=NoiseA_UVControl_Mark:********************
   channelNames {a=PanV, b=PanU, g=ScaleV, r=ScaleU}
   defaultValue=<rgba 1 1 0 0>
  }
  n16 = MaterialFunctionCall(Texture=@n14, UVControl_RGB=@n15, UVControl_A=@n15, "[Override]UVInput"=@n11) {materialFunction=&FXCommon_TextureSample}
  n17 = AppendVector(A=@n16.R, B=@n16.R)
  n18 = VectorParameter() {
   group=---NoiseA
   parameterName=NoiseA_ValueControl_Mark:--------------------
   sortPriority=33
   channelNames {a=StrengthV, b=StrengthU, g=StrengthOffsetV, r=StrengthOffsetU}
   defaultValue=<rgba -0.5 -0.5 0 0>
  }
  n19 = MaterialFunctionCall(
   Texture=@n2
   AdditionalUVOffset=LinearInterpolate(A=AppendVector(A=@n3, B=@n5), B=AppendVector(A=@n3, B=@n6), Alpha=Saturate(None=@n12))
   UVControl_RGB=@n13
   UVControl_A=@n13
   "[Override]UVInput"=Add(
    A=@n11
    B=MaterialFunctionCall!(NoiseValue=@n17, Control_RGB=@n18, Control_A=@n18) {materialFunction=/Game/Effects/Niagara/MasterMaterials/Functions/Common/FXCommon_NoiseControl}.NoiseValue
   )
  ) {materialFunction=&FXCommon_TextureSample}
  n20 = VectorParameter() {parameterName=clamp, defaultValue=<rgba 0 1 0 1>}
  n21 = Multiply(
   A=MaterialFunctionCall!(TexCoord=@n19.UV, ClampChannel=AppendVector(A=@n20.R, B=@n20.R)) {materialFunction=/Game/Effects/Fort_Effects/MaterialFunctions/MF_FX_TilingMethod}.Result
   B=@n19.RGB
  )
  n22 = VectorParameter() {parameterName=Refine2, defaultValue=<rgba 10 2000 0.3 0.5>}
  n23 = MaterialFunctionCall(Texture=@n21, RGB=@n22.RGB, A=@n22.RGB) {materialFunction=&MF_FX_Refine}
  n24 = VectorParameter() {parameterName=C, defaultValue=<rgba 0 0 0 1>}
  MaterialFunctionCall(Emissive=Multiply(A=@n1.RGB, B=Multiply(A=@n23.Result, B=@n24.RGB))) {materialFunction=&MF_FX_EmissiveAdjustments}
  n25 = MaterialFunctionCall(Random=Constant()) {materialFunction=&DitherTemporalAA_Fortnite}
  n26 = MaterialFunctionCall!(
   UorV=MaterialFunctionCall(Float2=TextureCoordinate()) {materialFunction=&BreakOutFloat2Components}.G
   Invert_EdgeControl=VectorParameter() {parameterName="Invert_EdgeControl (V3)", defaultValue=<rgba 0 1 1 0>}.RGB
   NoInvert_EdgeControl=VectorParameter() {parameterName="NoInvert_EdgeControl (V3)", defaultValue=<rgba 0 1 1 0>}.RGB
  ) {materialFunction=/Game/Effects/Niagara/MasterMaterials/Functions/Common/FXCommon_EdgeMask_Ex}
  n27 = VectorParameter() {parameterName=Refine, defaultValue=<rgba 3 1 0.7 0.5>}
  n28 = MaterialFunctionCall(Texture=@n21, RGB=@n27.RGB, A=@n27.RGB) {materialFunction=&MF_FX_Refine}
  n29 = MaterialFunctionCall(Vector 3=Subtract(A=ParticlePositionWS(), B=CameraPositionWS())) {materialFunction=/Engine/Functions/Engine_MaterialFunctions02/Utility/VectorLength}
  n30 = LinearInterpolate(A=Multiply(A=@n1.A, B=Multiply(A=@n26.Mask, B=@n28.Result)), Alpha=Saturate(None=Divide(A=@n29.V3 Length) {constB=4096})) {constB=0}
  n31 = Constant() {r=1}
  ShadingPathSwitch(
   Default=LinearInterpolate(A=Constant() {r=1}, B=ScalarParameter() {parameterName=R}, Alpha=LinearInterpolate(B=@n25.Result, Alpha=@n30))
   Forward=@n31
   Mobile=@n31
  )
 ]
}
