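# Translucent unlit Niagara ribbon material, apparently for finger trails in the UltraLight emote.
# Two textures share center-scaled UVs with separate pan/scale controls; A defaults to a soft smoke trail.
# Emissive adds refined texture B, base brightness and scaled, normalized texture A color, tinted by particle color.
# Opacity multiplies a separately refined texture B by desaturated, clamp-masked texture A, a V-axis edge mask and particle alpha.
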Material /BRCosmetics/Effects/Fort_Effects/Emotes/UltraLight/Material/M_FingerRibbon
graph {
 properties {bUsedWithNiagaraRibbons=true, blendMode=BLEND_Translucent, pixelDepthOffsetMode=PDOM_Legacy, shadingModel=MSM_Unlit, usedShadingModels=Unlit}
 nodes [
  n1 = VectorParameter() {
   group=---TextureB
   parameterName=TextureB_UVControl_Mark:********************
   channelNames {a=PanV, b=PanU, g=ScaleV, r=ScaleU}
   defaultValue=<rgba 1 1 0 0>
  }
  n2 = VectorParameter() {parameterName=TextureA_ScaleCenter, defaultValue=<rgba 0 0 0 1>}
  n3 = MaterialFunctionCall(UVs=TextureCoordinate(), Texture Scale=AppendVector(A=@n2.R, B=@n2.R)) {materialFunction=&ScaleUVsByCenter}
  n4 = MaterialFunctionCall(
   Texture=TextureObjectParameter() {group=---TextureB, parameterName=TextureB_Texture, sortPriority=31}
   UVControl_RGB=@n1
   UVControl_A=@n1
   "[Override]UVInput"=@n3.UVs
  ) {materialFunction=&FXCommon_TextureSample}
  n5 = VectorParameter() {parameterName=TextureB_Refine_Brightness, defaultValue=<rgba 10 2000 0.3 0.5>}
  n6 = Add(
   A=MaterialFunctionCall(Texture=@n4.R, RGB=@n5.RGB, A=@n5.RGB) {materialFunction=&MF_FX_Refine}.Result
   B=ScalarParameter() {parameterName=BaseBrightness}
  )
  n7 = VectorParameter() {
   group=---TextureA
   parameterName=TextureA_UVControl_Mark:********************
   channelNames {a=PanV, b=PanU, g=ScaleV, r=ScaleU}
   defaultValue=<rgba 1 1 0 0>
  }
  n8 = MaterialFunctionCall(
   Texture=TextureObjectParameter() {group=---TextureA, parameterName=TextureA_Texture, texture=/Game/Effects/Fort_Effects/Textures/Smoke/T_Smoke_Trail_Soft_2}
   UVControl_RGB=@n7
   UVControl_A=@n7
   "[Override]UVInput"=@n3.UVs
  ) {materialFunction=&FXCommon_TextureSample}
  n9 = Multiply(A=Normalize(VectorInput=@n8.RGB), B=ScalarParameter() {defaultValue=0.3, parameterName=NormalizedTextureColorScale})
  n10 = ParticleColor()
  MaterialFunctionCall(Emissive=Multiply(A=Add(A=@n6, B=@n9), B=@n10.RGB)) {materialFunction=&MF_FX_EmissiveAdjustments}
  n11 = VectorParameter() {parameterName=TextureB_Refine_Opacity, defaultValue=<rgba 3 1 0.7 0.5>}
  n12 = MaterialFunctionCall(Float2=TextureCoordinate()) {materialFunction=&BreakOutFloat2Components}
  n13 = VectorParameter() {parameterName=EC, defaultValue=<rgba 1 1 1 1>}
  n14 = Constant2Vector() {g=1}
  n15 = Desaturation(
   None=Multiply(
    A=MaterialFunctionCall!(TexCoord=@n3.UVs, ClampChannel=@n14) {materialFunction=/Game/Effects/Fort_Effects/MaterialFunctions/MF_FX_TilingMethod}.Result
    B=@n8.RGB
   )
   Fraction=Constant() {r=1}
  )
  Multiply(
   A=MaterialFunctionCall(Texture=@n4.R, RGB=@n11.RGB, A=@n11.RGB) {materialFunction=&MF_FX_Refine}.Result
   B=Multiply(
    A=MaterialFunctionCall!(UorV=@n12.G, EdgeControl=@n13.RGB) {materialFunction=/Game/Effects/Niagara/MasterMaterials/Functions/Common/FXCommon_EdgeMask_Simple}.Mask
    B=Multiply(A=@n10.A, B=@n15)
   )
  )
 ]
}
