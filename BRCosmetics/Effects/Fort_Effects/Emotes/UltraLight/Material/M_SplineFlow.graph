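# Translucent, unlit, two-sided material for Niagara mesh particles in the UltraLight emote, drawing flowing trails on SplineThicken geometry.
# A NiagaraCurveInput texture sampled along the strip drives width, opacity and emissive through separate channels.
# A panning trail texture, distorted by noise and offset by dynamic-parameter time, feeds separately refined color and opacity, tinted by particle color.
# Depth fade softens opacity and modulates optional refraction, which is neutral on forward and mobile paths.
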Material /BRCosmetics/Effects/Fort_Effects/Emotes/UltraLight/Material/M_SplineFlow
graph {
 properties {
  bTangentSpaceNormal=false
  bUsedWithNiagaraMeshParticles=true
  blendMode=BLEND_Translucent
  pixelDepthOffsetMode=PDOM_Legacy
  refractionMethod=RM_IndexOfRefraction
  shadingModel=MSM_Unlit
  twoSided=true
  usedShadingModels=Unlit
 }
 nodes [
  n1 = MaterialFunctionCall(XY=TextureCoordinate()) {materialFunction=&Swizzle}
  n2 = TextureObjectParameter() {
   group=---TextureA
   parameterName=NiagaraCurveInput
   samplerType=SAMPLERTYPE_LinearColor
   texture=/Game/Effects/Fort_Effects/Emotes/UltraLight/Texture/Curve/CLA_1x1
  }
  n3 = TextureSample(
   UVs=AppendVector(
    A=MaterialFunctionCall(Float2=@n1.YX) {materialFunction=&BreakOutFloat2Components}.R
    B=Constant() {r=0.5}
   )
   Tex=@n2
  )
  n4 = Multiply(A=@n3.R, B=ScalarParameter() {defaultValue=50, parameterName=Width})
  MaterialFunctionCall(WidthBase=@n4, WidthTip=@n4, FlattenPixelNormal=StaticBool() {value=true}) {materialFunction=&SplineThicken}
  n5 = TextureObjectParameter() {
   group=---TextureA
   parameterName=TextureA_Texture
   samplerType=SAMPLERTYPE_Grayscale
   texture=/Game/Effects/Fort_Effects/Textures/Smoke/T_Smoke_Trail_Soft_NoClamp
  }
  n6 = VectorParameter() {
   group=---TextureA
   parameterName=TextureA_UVControl_Mark:********************
   channelNames {a=PanV, b=PanU, g=ScaleV, r=ScaleU}
   defaultValue=<rgba 1 1 0 0>
  }
  n7 = Add(
   A=Time() {bOverride_Period=true, period=3333}
   B=DynamicParameter() {defaultValue=<rgba 1 5 1 1>, paramNames [TimeInput,Param2,Param3,Param4]}.TimeInput
  )
  n8 = TextureObjectParameter() {group=---NoiseA, parameterName=NoiseA_Texture, texture=/Game/Characters/Player/Male/Medium/Bodies/M_MED_Viper/Textures/Glitch/T_SoftNoise}
  n9 = VectorParameter() {
   group=---NoiseA
   parameterName=NoiseA_UVControl_Mark:********************
   channelNames {a=PanV, b=PanU, g=ScaleV, r=ScaleU}
   defaultValue=<rgba 1 1 0 0>
  }
  n10 = VectorParameter() {
   group=---NoiseA
   parameterName=NoiseA_ValueControl_Mark:--------------------
   sortPriority=33
   channelNames {a=StrengthV, b=StrengthU, g=StrengthOffsetV, r=StrengthOffsetU}
   defaultValue=<rgba -0.5 -0.5 0 0>
  }
  n11 = Add(
   A=@n1.YX
   B=MaterialFunctionCall!(
    NoiseValue=MaterialFunctionCall(Texture=@n8, UVControl_RGB=@n9, UVControl_A=@n9, "[Override]TimeInput"=@n7) {materialFunction=&FXCommon_TextureSample}.RG
    Control_RGB=@n10
    Control_A=@n10
   ) {materialFunction=/Game/Effects/Niagara/MasterMaterials/Functions/Common/FXCommon_NoiseControl}.NoiseValue
  )
  n12 = MaterialFunctionCall(Texture=@n5, UVControl_RGB=@n6, UVControl_A=@n6, "[Override]TimeInput"=@n7, "[Override]UVInput"=@n11) {materialFunction=&FXCommon_TextureSample}
  n13 = VectorParameter() {parameterName=Refine_Color, defaultValue=<rgba 3 1 0.7 0.5>}
  n14 = Add(A=MaterialFunctionCall(Texture=@n12.R, RGB=@n13.RGB, A=@n13.RGB) {materialFunction=&MF_FX_Refine}.Result)
  n15 = ParticleColor()
  MaterialFunctionCall(Emissive=Multiply(A=@n3.B, B=Multiply(A=@n14, B=@n15.RGB))) {materialFunction=&MF_FX_EmissiveAdjustments}
  n16 = MaterialFunctionCall!(TexCoord=@n11, ClampChannel=Constant2Vector() {g=1}) {materialFunction=/Game/Effects/Fort_Effects/MaterialFunctions/MF_FX_TilingMethod}
  n17 = VectorParameter() {parameterName=Refine_Opacity, defaultValue=<rgba 3 1 0.7 0.5>}
  n18 = Multiply(
   A=@n3.G
   B=MaterialFunctionCall(Texture=@n12.R, RGB=@n17.RGB, A=@n17.RGB) {materialFunction=&MF_FX_Refine}.Result
  )
  n19 = DepthFade(Opacity=Multiply(A=@n15.A, B=Multiply(A=@n16.Result, B=@n18)), FadeDistance=ScalarParameter() {defaultValue=10, parameterName=DF})
  n20 = LinearInterpolate(
   B=ScalarParameter() {parameterName=Refraction}
   Alpha=LinearInterpolate(
    B=MaterialFunctionCall(Alpha Threshold=@n19) {materialFunction=&DitherTemporalAA_Fortnite}.Result
    Alpha=@n19
   )
  ) {constA=1}
  n21 = Constant() {r=1}
  ShadingPathSwitch(Default=@n20, Deferred=@n20, Forward=@n21, Mobile=@n21)
  Transform(None=Constant3Vector() {constant=<rgba 0 0 1 0>}) {transformSourceType=TRANSFORMSOURCE_Instance}
 ]
}
