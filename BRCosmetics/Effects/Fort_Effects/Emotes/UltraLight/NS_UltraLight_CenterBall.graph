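# Niagara system for the UltraLight emote that draws a glowing energy ball with mesh beams linking it to both hands.
# Core and energy meshes, glow sprites and a flickering sprite follow a center point, while beams orient from sampled hand positions and scale with hand separation.
# An animation progress value drives fades, scaling and material timeline speed.
# Shared hue, saturation and value tint layers, with front-end and time-of-day brightness tweaks on some.
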NiagaraSystem /BRCosmetics/Effects/Fort_Effects/Emotes/UltraLight/NS_UltraLight_CenterBall
dependencies {
 usedDataInterfaces [&NiagaraDataInterfaceCurve, &NiagaraDataInterfaceColorCurve]
 usedDynamicInputs [
  &Add_Vector
  &Multiply_VectorByFloat
  &RotateVectorByQuaternion
  &Divide_Vector
 ]
 usedModules [
  &SystemState
  &SampleSkeletalMesh
  &self:SystemSpawnScript.SystemScriptSource.SystemScriptGraph.NiagaraNodeAssignment_0.SetVariables_#1
  &self:SystemSpawnScript.SystemScriptSource.SystemScriptGraph.NiagaraNodeAssignment_1.SetVariables_#2
  &self:SystemSpawnScript.SystemScriptSource.SystemScriptGraph.NiagaraNodeAssignment_2.SetVariables_#3
  &self:SystemSpawnScript.SystemScriptSource.SystemScriptGraph.NiagaraNodeAssignment_3.SetVariables_#4
 ]
}
emitters {
 Core {
  emitterUpdateScript [
   &EmitterState(
    Inactive Response=::NewEnumerator1
    Life Cycle Mode=::NewEnumerator1
    Loop Behavior=::NewEnumerator1
    Loop Duration Mode=::NewEnumerator1
    UseLoopDelay=false
   )
   &SpawnBurst_Instantaneous(Spawn Count=1, Use Spawn Probability=false)
  ]
  particleSpawnScript [
   &InitializeParticle(
    Color=<rgba 1 0 0.05 1>
    Mesh Scale Mode=::NewEnumerator3
    Mesh Uniform Scale=0.07
    Sprite Size Mode=::NewEnumerator3
    Uniform Sprite Size=0
    UsePositionOffset=false
    Write Material Random=true
   )
   &CameraOffset()
  ]
  particleUpdateScript [
   &ParticleState(Kill Particles When Lifetime Has Elapsed=false)
   /SetVariables(Particles.Position=$System.CenterPos) {isSetParameters=true}
   &ScaleColor(
    Scale Alpha=&FloatFromCurve(
     FloatCurve=DataInterface() {propertyValues {Curve {keys [{interpMode=RCIM_Cubic, leaveTangent=1.3475, tangentMode=RCTM_Break, time=0, value=0}, <at 1 1>]}}}
     CurveIndex=$System.AnimProgressInput
    )
    ScaleA=true
    ScaleRGB=true
   )
   &ScaleMeshSize(Scale Factor=&VectorFromFloat(Value=&FloatFromCurve(FloatCurve=DataInterface() {propertyValues {Curve {keys [<at 0 1>,<at 0.8 1>,<at 0.85 1.3>]}}}, CurveIndex=$System.AnimProgressInput)))
   &self:ColorAdjust(Hue:NiagaraFloat=$System.Hue, Saturation:NiagaraFloat=$System.Sat, Value:NiagaraFloat=$System.Value) {module=ColorAdjust}
  ]
  rendererClasses [&NiagaraMeshRendererProperties]
  renderers [{rendererClass=&NiagaraMeshRendererProperties, rendererIndex=0}]
 }
 CoreGlow {
  emitterUpdateScript [
   &EmitterState(
    Inactive Response=::NewEnumerator1
    Life Cycle Mode=::NewEnumerator1
    Loop Behavior=::NewEnumerator1
    Loop Duration Mode=::NewEnumerator1
    UseLoopDelay=false
   )
   &SpawnBurst_Instantaneous(Spawn Count=1, Use Spawn Probability=false)
   /SetVariables(Emitter.Refine:Vector4f=<17 5 1.6 0.005>) {isSetParameters=true, module=SetVariables_#5}
  ]
  particleSpawnScript [
   &InitializeParticle(Color=<rgba 500 0 24.9996 1>, Sprite Size Mode=::NewEnumerator3, UsePositionOffset=false, Write Material Random=true)
   &CameraOffset()
  ]
  particleUpdateScript [
   &ParticleState(Kill Particles When Lifetime Has Elapsed=false)
   /SetVariables(Particles.Position=$System.CenterPos) {isSetParameters=true}
   &ScaleColor(
    Scale Alpha=&FloatFromCurve(
     FloatCurve=DataInterface() {propertyValues {Curve {keys [{interpMode=RCIM_Cubic, tangentMode=RCTM_Break, time=0.05, value=0}, <at 1 1>]}}}
     CurveIndex=$System.AnimProgressInput
    )
    ScaleA=true
    ScaleRGB=true
   )
   &self:ColorAdjust(
    Hue:NiagaraFloat=$System.Hue
    Saturation:NiagaraFloat=$System.Sat
    Value:NiagaraFloat=&MakeCustomFloatFromBool(Bool=$User.bIsFrontEnd, True Float=2, False Float=$System.Value)
   ) {module=ColorAdjust}
  ]
  rendererClasses [&NiagaraSpriteRendererProperties]
  renderers [
   {rendererClass=&NiagaraSpriteRendererProperties}
   {rendererClass=&NiagaraSpriteRendererProperties, rendererIndex=1}
  ]
 }
 CoreGlow001 {
  emitterUpdateScript [
   &EmitterState(
    Inactive Response=::NewEnumerator1
    Life Cycle Mode=::NewEnumerator1
    Loop Behavior=::NewEnumerator1
    Loop Duration Mode=::NewEnumerator1
    UseLoopDelay=false
   )
   &SpawnBurst_Instantaneous(Spawn Count=1, Use Spawn Probability=false)
   /SetVariables(Emitter.Refine:Vector4f=<15 15 2 1.5>, Emitter.Refine2:Vector4f=<1 0.2 0 0>) {isSetParameters=true, module=SetVariables_#5}
  ]
  particleSpawnScript [
   &InitializeParticle(Color=<rgba 1 0 0.05 0.3>, Sprite Size Mode=::NewEnumerator3, Uniform Sprite Size=100, UsePositionOffset=false, Write Material Random=true)
   &CameraOffset()
  ]
  particleUpdateScript [
   &ParticleState(Kill Particles When Lifetime Has Elapsed=false)
   /SetVariables(Particles.Position=$System.CenterPos) {isSetParameters=true}
   &ScaleColor(
    Scale Alpha=&FloatFromCurve(
     FloatCurve=DataInterface() {propertyValues {Curve {keys [{interpMode=RCIM_Cubic, leaveTangent=3.3849, tangentMode=RCTM_Break, time=0, value=0}, <at 1 1>]}}}
     CurveIndex=$System.AnimProgressInput
    )
    ScaleA=true
    ScaleRGB=true
   )
   &self:ColorAdjust(
    Hue:NiagaraFloat=$System.Hue
    Saturation:NiagaraFloat=$System.Sat
    Value:NiagaraFloat=&Multiply_Float(A=$System.Value, B=1)
   ) {module=ColorAdjust}
  ]
  rendererClasses [&NiagaraSpriteRendererProperties]
  renderers [
   {rendererClass=&NiagaraSpriteRendererProperties}
   {rendererClass=&NiagaraSpriteRendererProperties, rendererIndex=1}
  ]
 }
 DarkGlow001 {
  emitterUpdateScript [
   &EmitterState(
    Inactive Response=::NewEnumerator1
    Life Cycle Mode=::NewEnumerator1
    Loop Behavior=::NewEnumerator1
    Loop Duration Mode=::NewEnumerator1
    UseLoopDelay=false
   )
   &SpawnBurst_Instantaneous(Spawn Count=1, Use Spawn Probability=false)
   /SetVariables(Emitter.HueShift=-0.13, Emitter.Refine:Vector4f=<3 1 3 1>) {isSetParameters=true}
  ]
  particleSpawnScript [
   &InitializeParticle(
    Color=<rgba 0.042 0.021 0.022 1.02>
    Sprite Size Mode=::NewEnumerator3
    Uniform Sprite Size=110
    UsePositionOffset=false
    Write Material Random=true
   )
   &CameraOffset()
  ]
  particleUpdateScript [
   &ParticleState(Kill Particles When Lifetime Has Elapsed=false)
   /SetVariables(Particles.Position=$System.CenterPos) {isSetParameters=true}
   &ScaleColor(
    Scale Alpha=&FloatFromCurve(
     FloatCurve=DataInterface() {propertyValues {Curve {keys [{interpMode=RCIM_Cubic, leaveTangent=1.9463, tangentMode=RCTM_Break, time=0, value=0}, <at 1 1>]}}}
     CurveIndex=$System.AnimProgressInput
     Scale Curve=&MakeCustomFloatFromBool(Bool=$User.bIsFrontEnd, True Float=0.5, False Float=1)
    )
    ScaleA=true
    ScaleRGB=true
   )
   &self:ColorAdjust(
    Hue:NiagaraFloat=$System.Hue
    Saturation:NiagaraFloat=$System.Sat
    Value:NiagaraFloat=&MakeCustomFloatFromBool(
     Bool=$User.bIsFrontEnd
     True Float=3
     False Float=&Multiply_Float(
      A=$System.Value
      B=&FloatFromCurve(
       FloatCurve=DataInterface() {propertyValues {Curve {keys [<at 0 1.5>,<at 0.0833 1.5>,<at 0.1666 1.5>,<at 0.27 1>,<at 0.5 1>,<at 0.666 1>,<at 0.7708 1>,<at 0.8333 1>,<at 0.875 1>,<at 1 1.5>]}}}
       CurveIndex=Linked() {linkedVariable {name=NPC.FortniteNPC.TimeOfDay_0-1, type=NiagaraFloat}}
      )
     )
    )
   ) {module=ColorAdjust}
  ]
  rendererClasses [&NiagaraSpriteRendererProperties]
  renderers [{rendererClass=&NiagaraSpriteRendererProperties}]
 }
 Empty {
  emitterUpdateScript [
   &EmitterState(
    Inactive Response=::NewEnumerator1
    Life Cycle Mode=::NewEnumerator1
    Loop Behavior=::NewEnumerator1
    Loop Duration Mode=::NewEnumerator1
    UseLoopDelay=false
   )
   &SpawnBurst_Instantaneous(Use Spawn Probability=false, UseLoopCountLimit=false)
  ]
  particleSpawnScript [&InitializeParticle(
    Color=<rgba 4 0 0.2 1>
    Lifetime=5
    Sprite Size=<300 264.2582>
    Sprite Size Mode=::NewEnumerator4
    UsePositionOffset=false
    Write Material Random=true
   )]
  particleUpdateScript [
   &ParticleState(Kill Particles When Lifetime Has Elapsed=false, Let Infinitely Lived Particles Die When Emitter Deactivates=true)
   /SetVariables(Particles.Position=$System.CenterPos) {isSetParameters=true}
   &DynamicMaterialParameters(
    Index 0 Param 1=0.1
    Index 0 Param 2=0.025
    Index 0 Param 3=-0.955
    Index 0 Param 4=100
    Module.Param0WriteEnabled=true
    Module.Param1WriteEnabled=true
    Module.Param2WriteEnabled=true
    Module.Param3WriteEnabled=true
   )
   &ScaleColor(
    Scale Alpha=&Waveform("Waveform [1]"=::NewEnumerator12, "[1] Phase Scale"=15, "[1] Frequency"=1, "[1] Amplitude Min/Max"=<0 1>)
    ScaleA=true
    ScaleRGB=true
   )
   &ScaleColor(
    Scale RGB=&VectorFromFloat(Value=&FloatFromCurve(
      FloatCurve=DataInterface() {propertyValues {Curve {keys [
          {interpMode=RCIM_Cubic, leaveTangent=1.721, tangentMode=RCTM_Break, time=0, value=0}
          {arriveTangent=2.4433, interpMode=RCIM_Cubic, leaveTangent=2.4433, tangentMode=RCTM_User, time=0.5, value=0.675}
          {arriveTangent=0.0018, interpMode=RCIM_Cubic, leaveTangent=0.0018, tangentMode=RCTM_User, time=0.7, value=1.625}
          {arriveTangent=-5.392, interpMode=RCIM_Cubic, tangentMode=RCTM_Break, time=1, value=1}
         ]}}}
      CurveIndex=$System.AnimProgressInput
     ))
    ScaleA=true
    ScaleRGB=true
   ) {module=ScaleColor001}
   &self:ColorAdjust(Hue:NiagaraFloat=$System.Hue, Saturation:NiagaraFloat=$System.Sat, Value:NiagaraFloat=$System.Value) {module=ColorAdjust}
   &ScaleSpriteSize(
    Scale Factor=&MakeVector2D(
     X=&FloatFromCurve(
      FloatCurve=DataInterface() {propertyValues {Curve {keys [{interpMode=RCIM_Cubic, tangentMode=RCTM_Break, time=0, value=1.6}, <at 1 1.1 RCIM_Cubic>]}}}
      CurveIndex=$System.AnimProgressInput
     )
     Y=1
    )
    Scale Sprite Size Mode=::NewEnumerator2
   )
  ]
  rendererClasses [&NiagaraSpriteRendererProperties]
  renderers [{rendererClass=&NiagaraSpriteRendererProperties}]
 }
 EnergyCenter {
  emitterUpdateScript [
   &EmitterState(
    Inactive Response=::NewEnumerator1
    Life Cycle Mode=::NewEnumerator1
    Loop Behavior=::NewEnumerator1
    Loop Duration Mode=::NewEnumerator1
    UseLoopDelay=false
   )
   &SpawnBurst_Instantaneous(Use Spawn Probability=false, UseLoopCountLimit=false)
  ]
  particleSpawnScript [&InitializeParticle(
    Color=<rgba 15 15 15 1>
    Lifetime=5
    Mesh Scale Mode=::NewEnumerator3
    Mesh Uniform Scale=0.13
    UsePositionOffset=false
    Write Material Random=true
   )]
  particleUpdateScript [
   &ParticleState(Kill Particles When Lifetime Has Elapsed=false, Let Infinitely Lived Particles Die When Emitter Deactivates=true)
   &Timeline(
    Play Rate=&FloatFromCurve(FloatCurve=DataInterface() {propertyValues {Curve {keys [<at 0 1.25>,<at 0.75 1.25>,<at 0.8 2>]}}}, CurveIndex=$System.AnimProgressInput)
    Rewind Mode:ENiagara_TimelineMode=::NewEnumerator1
    Timeline End=9999999
   )
   &DynamicMaterialParameters(
    Index 0 Param 1=$Particles.Timeline.Time
    Module.Param0WriteEnabled=true
    Module.Param1WriteEnabled=true
    Module.Param2WriteEnabled=true
    Module.Param3WriteEnabled=true
   )
   /SetVariables(Particles.Position=$System.CenterPos) {isSetParameters=true}
   &ScaleColor(
    Scale Alpha=&FloatFromCurve(
     FloatCurve=DataInterface() {propertyValues {Curve {keys [{interpMode=RCIM_Cubic, leaveTangent=2.6493, tangentMode=RCTM_Break, time=0, value=0}, <at 1 1>]}}}
     CurveIndex=$System.AnimProgressInput
    )
    Scale RGB=&MakeVectorFromLinearColorRGB(Color=&ColorFromCurve(ColorCurve=DataInterface() {propertyValues {
        AlphaCurve {keys [<at 0 1 RCIM_Cubic>,<at 1 1 RCIM_Cubic>]}
        BlueCurve {keys [<at 0 0.05 RCIM_Cubic>,<at 1 0.05 RCIM_Cubic>]}
        GreenCurve {keys [<at 0 0 RCIM_Cubic>,<at 1 0 RCIM_Cubic>]}
        RedCurve {keys [<at 0 1 RCIM_Cubic>,<at 1 1 RCIM_Cubic>]}
       }}))
    ScaleA=true
    ScaleRGB=true
   )
   &ScaleMeshSize(Scale Factor=&VectorFromFloat(Value=&FloatFromCurve(FloatCurve=DataInterface() {propertyValues {Curve {keys [<at 0 1>,<at 0.8 1>,<at 0.85 1.3>]}}}, CurveIndex=$System.AnimProgressInput)))
   &self:ColorAdjust(Hue:NiagaraFloat=$System.Hue, Saturation:NiagaraFloat=$System.Sat, Value:NiagaraFloat=$System.Value) {module=ColorAdjust}
  ]
  rendererClasses [&NiagaraMeshRendererProperties]
  renderers [{rendererClass=&NiagaraMeshRendererProperties, rendererIndex=0}]
 }
 EnergyCenter001 {
  emitterUpdateScript [
   &EmitterState(
    Inactive Response=::NewEnumerator1
    Life Cycle Mode=::NewEnumerator1
    Loop Behavior=::NewEnumerator1
    Loop Duration Mode=::NewEnumerator1
    UseLoopDelay=false
   )
   &SpawnBurst_Instantaneous(Use Spawn Probability=false, UseLoopCountLimit=false)
   /SetVariables(Emitter.HueS:Vector4f=<0.2 0.005 0 0>) {isSetParameters=true}
  ]
  particleSpawnScript [&InitializeParticle(
    Color=<rgba 1 1 1 2>
    Lifetime=5
    Mesh Scale Mode=::NewEnumerator3
    Mesh Uniform Scale=0.25
    UsePositionOffset=false
    Write Material Random=true
   )]
  particleUpdateScript [
   &ParticleState(Kill Particles When Lifetime Has Elapsed=false, Let Infinitely Lived Particles Die When Emitter Deactivates=true)
   &Timeline(
    Play Rate=&FloatFromCurve(FloatCurve=DataInterface() {propertyValues {Curve {keys [<at 0 1>,<at 0.75 1>,<at 0.8 1.75>]}}}, CurveIndex=$System.AnimProgressInput)
    Rewind Mode:ENiagara_TimelineMode=::NewEnumerator1
    Timeline End=9999999
   )
   &DynamicMaterialParameters(
    Index 0 Param 1=$Particles.Timeline.Time
    Index 3 Param 1=&FloatFromCurve(
     FloatCurve=DataInterface() {propertyValues {Curve {keys [
         {interpMode=RCIM_Cubic, leaveTangent=6.5468, tangentMode=RCTM_Break, time=0, value=-2}
         {arriveTangent=0.2104, interpMode=RCIM_Cubic, tangentMode=RCTM_Break, time=0.75, value=0.5}
        ]}}}
     CurveIndex=$System.AnimProgressInput
    )
    Module.Param0WriteEnabled=true
    Module.Param0WriteEnabled3=true
    Module.Param1WriteEnabled=true
    Module.Param1WriteEnabled3=true
    Module.Param2WriteEnabled=true
    Module.Param2WriteEnabled3=true
    Module.Param3WriteEnabled=true
    Module.Param3WriteEnabled3=true
    Write Parameter Index 3=true
   )
   /SetVariables(Particles.Position=$System.CenterPos) {isSetParameters=true}
   &ScaleColor(
    Scale Alpha=&FloatFromCurve(
     FloatCurve=DataInterface() {propertyValues {Curve {keys [{interpMode=RCIM_Cubic, leaveTangent=0.0606, tangentMode=RCTM_Break, time=0, value=0}, <at 1 0.65>]}}}
     CurveIndex=$System.AnimProgressInput
     Scale Curve=&MakeCustomFloatFromBool(Bool=$User.bIsFrontEnd, True Float=0.5, False Float=1)
    )
    Scale RGB=&MakeVectorFromLinearColorRGB(Color=&ColorFromCurve(ColorCurve=DataInterface() {propertyValues {
        AlphaCurve {keys [<at 0 1 RCIM_Cubic>,<at 1 1 RCIM_Cubic>]}
        BlueCurve {keys [<at 0 0.0833 RCIM_Cubic>,<at 1 0.0833 RCIM_Cubic>]}
        GreenCurve {keys [<at 0 0 RCIM_Cubic>,<at 1 0 RCIM_Cubic>]}
        RedCurve {keys [<at 0 1 RCIM_Cubic>,<at 1 1 RCIM_Cubic>]}
       }}))
    ScaleA=true
    ScaleRGB=true
   )
   &ScaleMeshSize(Scale Factor=&MakeVector(
     X=1
     Y=&FloatFromCurve(
      FloatCurve=DataInterface() {propertyValues {Curve {keys [{interpMode=RCIM_Cubic, tangentMode=RCTM_Break, time=0, value=3}, {interpMode=RCIM_Cubic, tangentMode=RCTM_Break, time=0.3, value=1}]}}}
      CurveIndex=$System.AnimProgressInput
     )
     Z=&FloatFromCurve(
      FloatCurve=DataInterface() {propertyValues {Curve {keys [{interpMode=RCIM_Cubic, tangentMode=RCTM_Break, time=0, value=3}, {interpMode=RCIM_Cubic, tangentMode=RCTM_Break, time=0.3, value=1}]}}}
      CurveIndex=$System.AnimProgressInput
     )
    ))
   &ScaleMeshSize(Scale Factor=&VectorFromFloat(Value=&FloatFromCurve(FloatCurve=DataInterface() {propertyValues {Curve {keys [<at 0 1>,<at 0.8 1>,<at 0.85 1.3>]}}}, CurveIndex=$System.AnimProgressInput))) {module=ScaleMeshSize001}
   &self:ColorAdjust(
    Hue:NiagaraFloat=$System.Hue
    Saturation:NiagaraFloat=$System.Sat
    Value:NiagaraFloat=&MakeCustomFloatFromBool(
     Bool=$User.bIsFrontEnd
     True Float=3
     False Float=&Multiply_Float(
      A=$System.Value
      B=&FloatFromCurve(
       FloatCurve=DataInterface() {propertyValues {Curve {keys [<at 0 1.5>,<at 0.1666 1.1>,<at 0.27 1>,<at 0.5 1>,<at 0.666 1>,<at 0.7708 1>,<at 0.8333 1>,<at 0.875 1>,<at 1 1.5>]}}}
       CurveIndex=Linked() {linkedVariable {name=NPC.FortniteNPC.TimeOfDay_0-1, type=NiagaraFloat}}
      )
     )
    )
   ) {module=ColorAdjust}
  ]
  rendererClasses [&NiagaraMeshRendererProperties]
  renderers [{rendererClass=&NiagaraMeshRendererProperties, rendererIndex=0}]
 }
 Spline {
  emitterUpdateScript [
   &EmitterState(
    Inactive Response=::NewEnumerator1
    Life Cycle Mode=::NewEnumerator1
    Loop Behavior=::NewEnumerator1
    Loop Duration Mode=::NewEnumerator1
    UseLoopDelay=false
   )
   &SpawnBurst_Instantaneous(Use Spawn Probability=false, UseLoopCountLimit=false)
  ]
  particleSpawnScript [&InitializeParticle(Color=<rgba 0.8 0.024 0.0628 1>, Lifetime=5, Mesh Scale Mode=::NewEnumerator3, UsePositionOffset=false, Write Material Random=true)]
  particleUpdateScript [
   &ParticleState(Kill Particles When Lifetime Has Elapsed=false, Let Infinitely Lived Particles Die When Emitter Deactivates=true)
   /SetVariables(Particles.Position:Vector3f=$System.HandL_Pos) {isSetParameters=true}
   /SetVariables(Particles.Position2:Vector3f=$System.HandR_Pos) {isSetParameters=true}
   &ScaleMeshSize(Scale Factor=&MakeVector(
     X=1
     Y=1
     Z=&Divide_Float(
      A=&VectorLength(Vector=&Subtract_Vector(A=$System.HandL_Pos, B=$System.HandR_Pos))
      B=20
     )
    ))
   /SetVariables(Particles.MeshOrientation=&FindQuatBetweenTwoVectors(
     From Vector=<0 0 1>
     To Vector=&NormalizeVector(
      Vector To Normalize=&Subtract_Vector(A=$System.CenterPos, B=$System.HandL_Pos)
      Safe Normalize=true
      Safe Fallback Threshold=0
      Safe Fallback Vector=<0 0 0.001>
     )
    )) {isSetParameters=true}
   /SetVariables(Particles.MeshOrientation2:Quat4f=&FindQuatBetweenTwoVectors(
     From Vector=<0 0 1>
     To Vector=&NormalizeVector(
      Vector To Normalize=&Subtract_Vector(A=$System.CenterPos, B=$System.HandR_Pos)
      Safe Normalize=true
      Safe Fallback Threshold=0
      Safe Fallback Vector=<0 0 0.001>
     )
    )) {isSetParameters=true}
   &ScaleColor(
    Scale Alpha=&FloatFromCurve(FloatCurve=DataInterface() {propertyValues {Curve {keys [<at 0 0>,<at 1 1>]}}}, CurveIndex=$System.AnimProgressInput)
    ScaleA=true
    ScaleRGB=true
   )
   &Timeline(
    Play Rate=&FloatFromCurve(
     FloatCurve=DataInterface() {propertyValues {Curve {keys [
         <at 0 0>
         {interpMode=RCIM_Cubic, tangentMode=RCTM_Break, time=0.75, value=0}
         {interpMode=RCIM_Cubic, tangentMode=RCTM_Break, time=0.8, value=0.75}
        ]}}}
     CurveIndex=$System.AnimProgressInput
    )
    Rewind Mode:ENiagara_TimelineMode=::NewEnumerator1
    Timeline End=9999999
   )
   &DynamicMaterialParameters(
    Index 0 Param 1=$Particles.Timeline.Time
    Module.Param0WriteEnabled=true
    Module.Param1WriteEnabled=true
    Module.Param2WriteEnabled=true
    Module.Param3WriteEnabled=true
   )
   &self:ColorAdjust(Hue:NiagaraFloat=$System.Hue, Saturation:NiagaraFloat=$System.Sat, Value:NiagaraFloat=$System.Value) {module=ColorAdjust}
  ]
  rendererClasses [&NiagaraMeshRendererProperties]
  renderers [
   {rendererClass=&NiagaraMeshRendererProperties, rendererIndex=0}
   {rendererClass=&NiagaraMeshRendererProperties, rendererIndex=1}
  ]
 }
}
summary {
 systemName=NS_UltraLight_CenterBall
 emitters [
  {emitterName=DarkGlow001, rendererClasses [&NiagaraSpriteRendererProperties]}
  {emitterName=Core, rendererClasses [&NiagaraMeshRendererProperties]}
  {emitterName=CoreGlow, rendererClasses [&NiagaraSpriteRendererProperties]}
  {emitterName=CoreGlow001, rendererClasses [&NiagaraSpriteRendererProperties]}
  {emitterName=EnergyCenter, rendererClasses [&NiagaraMeshRendererProperties]}
  {emitterName=EnergyCenter001, rendererClasses [&NiagaraMeshRendererProperties]}
  {emitterName=Spline, rendererClasses [&NiagaraMeshRendererProperties]}
  {emitterName=Empty, rendererClasses [&NiagaraSpriteRendererProperties]}
 ]
 userVariables [
  User.BallProgress {type=NiagaraFloat, defaultValue=0, description=""}
  User.bIsFrontEnd {type=NiagaraBool}
  User.ForHandSample {
   type=NiagaraDataInterfaceSkeletalMesh
   defaultValue {
    struct=NiagaraExt_VariableValue_DataInterface
    value {dataInterface=&self:NiagaraDataInterfaceSkeletalMesh_0}
   }
  }
 ]
}
