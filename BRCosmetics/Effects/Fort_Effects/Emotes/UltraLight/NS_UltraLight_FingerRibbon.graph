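# Niagara system for the UltraLight emote that draws glowing finger ribbon trails as the effect moves.
# Two near-identical emitters spawn short-lived pink-red ribbons per distance travelled, differing only in curl noise quality setting.
# Curves swell then taper width, fade alpha in and out scaled by User.RibbonOpacity, flash RGB bright at birth, ramp curl noise over life and drive a dynamic material parameter.
# A local ColorAdjust module applies system-level hue, saturation and value.
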NiagaraSystem /BRCosmetics/Effects/Fort_Effects/Emotes/UltraLight/NS_UltraLight_FingerRibbon
dependencies {
 usedDataInterfaces [&NiagaraDataInterfaceCurve]
 usedDynamicInputs [&MakeCustomFloatFromBool]
 usedModules [
  &SystemState
  &self:SystemSpawnScript.SystemScriptSource.SystemScriptGraph.NiagaraNodeAssignment_0.SetVariables_#1
 ]
}
emitters {
 Empty {
  emitterUpdateScript [
   &EmitterState()
   &SpawnPerUnit(Module.Use Max Movement Threshold=true, Module.Use Movement Tolerance=false, Spawn Spacing=4, Use Spawn Probability=false)
  ]
  particleSpawnScript [
   &InitializeRibbon(Color=<rgba 1 0.03 0.0704 1>, Lifetime=0.75, Ribbon Width=12, UsePositionOffset=false, Write Material Random=true)
   /SetVariables(Particles.SpawnT:NiagaraFloat=&Divide_Float(A=$User.RibbonOpacity, B=2)) {isSetParameters=true, module=SetVariables_#2}
  ]
  particleUpdateScript [
   &ParticleState()
   &ScaleRibbonWidth(Ribbon Width Scale=&FloatFromCurve(FloatCurve=DataInterface() {propertyValues {Curve {keys [
         {arriveTangent=2.5155, interpMode=RCIM_Cubic, leaveTangent=2.5155, tangentMode=RCTM_User, time=0, value=0.5}
         {interpMode=RCIM_Cubic, tangentMode=RCTM_Break, time=0.3, value=1}
         {arriveTangent=-2.0725, interpMode=RCIM_Cubic, leaveTangent=-2.0725, tangentMode=RCTM_User, time=1, value=0}
        ]}}}))
   &ScaleColor(
    Scale Alpha=&FloatFromCurve(
     FloatCurve=DataInterface() {propertyValues {Curve {keys [
         {arriveTangent=7.8317, interpMode=RCIM_Cubic, leaveTangent=7.8317, tangentMode=RCTM_User, time=0, value=0}
         {interpMode=RCIM_Cubic, tangentMode=RCTM_User, time=0.3, value=1}
         {interpMode=RCIM_Cubic, tangentMode=RCTM_User, time=1, value=0}
        ]}}}
     Scale Curve=$Particles.SpawnT
    )
    Scale RGB=&VectorFromFloat(Value=&FloatFromCurve(FloatCurve=DataInterface() {propertyValues {Curve {keys [<at 0 10>,<at 0.2 1>]}}}))
    ScaleA=true
    ScaleRGB=true
   )
   &CurlNoiseForce(
    Noise Frequency=100
    Noise Quality / Cost=::NewEnumerator2
    Noise Strength=&FloatFromCurve(
     FloatCurve=DataInterface() {propertyValues {Curve {keys [{interpMode=RCIM_Cubic, tangentMode=RCTM_User, time=0, value=0}, {interpMode=RCIM_Cubic, tangentMode=RCTM_User, time=1, value=1}]}}}
     Scale Curve=300
    )
    Pan Noise=true
    Pan Noise Field=<10 0 0>
   )
   &SolveForcesAndVelocity(Clamp Velocity=false, Limit Acceleration=false)
   &DynamicMaterialParameters(
    Index 0 Param 1=&FloatFromCurve(FloatCurve=DataInterface() {propertyValues {Curve {keys [
         {arriveTangent=4.1785, interpMode=RCIM_Cubic, leaveTangent=4.1785, tangentMode=RCTM_User, time=0, value=0}
         {arriveTangent=0.9995, interpMode=RCIM_Cubic, leaveTangent=0.9995, tangentMode=RCTM_User, time=0.2223, value=0.6825}
         {arriveTangent=-0.0405, interpMode=RCIM_Cubic, leaveTangent=-0.0405, tangentMode=RCTM_User, time=1, value=1}
        ]}}})
    Module.Param0WriteEnabled=true
    Module.Param1WriteEnabled=true
    Module.Param2WriteEnabled=true
    Module.Param3WriteEnabled=true
   )
   &self:ColorAdjust(Hue:NiagaraFloat=$System.Hue, Saturation:NiagaraFloat=$System.Sat, Value:NiagaraFloat=$System.Value) {module=ColorAdjust}
  ]
  rendererClasses [&NiagaraRibbonRendererProperties]
  renderers [{rendererClass=&NiagaraRibbonRendererProperties}]
 }
 Empty001 {
  emitterUpdateScript [
   &EmitterState()
   &SpawnPerUnit(Module.Use Max Movement Threshold=true, Module.Use Movement Tolerance=false, Spawn Spacing=4, Use Spawn Probability=false)
  ]
  particleSpawnScript [
   &InitializeRibbon(Color=<rgba 1 0.03 0.0704 1>, Lifetime=0.75, Ribbon Width=12, UsePositionOffset=false, Write Material Random=true)
   /SetVariables(Particles.SpawnT:NiagaraFloat=&Divide_Float(A=$User.RibbonOpacity, B=2)) {isSetParameters=true, module=SetVariables_#2}
  ]
  particleUpdateScript [
   &ParticleState()
   &ScaleRibbonWidth(Ribbon Width Scale=&FloatFromCurve(FloatCurve=DataInterface() {propertyValues {Curve {keys [
         {arriveTangent=2.5155, interpMode=RCIM_Cubic, leaveTangent=2.5155, tangentMode=RCTM_User, time=0, value=0.5}
         {interpMode=RCIM_Cubic, tangentMode=RCTM_Break, time=0.3, value=1}
         {arriveTangent=-2.0725, interpMode=RCIM_Cubic, leaveTangent=-2.0725, tangentMode=RCTM_User, time=1, value=0}
        ]}}}))
   &ScaleColor(
    Scale Alpha=&FloatFromCurve(
     FloatCurve=DataInterface() {propertyValues {Curve {keys [
         {arriveTangent=7.8317, interpMode=RCIM_Cubic, leaveTangent=7.8317, tangentMode=RCTM_User, time=0, value=0}
         {interpMode=RCIM_Cubic, tangentMode=RCTM_User, time=0.3, value=1}
         {interpMode=RCIM_Cubic, tangentMode=RCTM_User, time=1, value=0}
        ]}}}
     Scale Curve=$Particles.SpawnT
    )
    Scale RGB=&VectorFromFloat(Value=&FloatFromCurve(FloatCurve=DataInterface() {propertyValues {Curve {keys [<at 0 10>,<at 0.2 1>]}}}))
    ScaleA=true
    ScaleRGB=true
   )
   &CurlNoiseForce(
    Noise Frequency=100
    Noise Strength=&FloatFromCurve(
     FloatCurve=DataInterface() {propertyValues {Curve {keys [{interpMode=RCIM_Cubic, tangentMode=RCTM_User, time=0, value=0}, {interpMode=RCIM_Cubic, tangentMode=RCTM_User, time=1, value=1}]}}}
     Scale Curve=300
    )
    Pan Noise=true
    Pan Noise Field=<10 0 0>
   )
   &SolveForcesAndVelocity(Clamp Velocity=false, Limit Acceleration=false)
   &DynamicMaterialParameters(
    Index 0 Param 1=&FloatFromCurve(FloatCurve=DataInterface() {propertyValues {Curve {keys [
         {arriveTangent=4.1785, interpMode=RCIM_Cubic, leaveTangent=4.1785, tangentMode=RCTM_User, time=0, value=0}
         {arriveTangent=0.9995, interpMode=RCIM_Cubic, leaveTangent=0.9995, tangentMode=RCTM_User, time=0.2223, value=0.6825}
         {arriveTangent=-0.0405, interpMode=RCIM_Cubic, leaveTangent=-0.0405, tangentMode=RCTM_User, time=1, value=1}
        ]}}})
    Module.Param0WriteEnabled=true
    Module.Param1WriteEnabled=true
    Module.Param2WriteEnabled=true
    Module.Param3WriteEnabled=true
   )
   &self:ColorAdjust(Hue:NiagaraFloat=$System.Hue, Saturation:NiagaraFloat=$System.Sat, Value:NiagaraFloat=$System.Value) {module=ColorAdjust}
  ]
  rendererClasses [&NiagaraRibbonRendererProperties]
  renderers [{rendererClass=&NiagaraRibbonRendererProperties}]
 }
}
summary {
 systemName=NS_UltraLight_FingerRibbon
 emitters [
  {emitterName=Empty, rendererClasses [&NiagaraRibbonRendererProperties]}
  {emitterName=Empty001, rendererClasses [&NiagaraRibbonRendererProperties]}
 ]
 userVariables [
  User.bIsFrontEnd {type=NiagaraBool}
  User.RibbonOpacity {type=NiagaraFloat, defaultValue=1, description=""}
 ]
}
