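# Niagara system for the UltraLight emote that puts a pink-red energy glow on the character's hands.
# A skeletal mesh sample supplies left and right hand positions, which burst-spawned particles store as Position and Position2 in emitters with paired renderers.
# Mesh cores, glow sprites, a decal-like glow and a flickering light build with an animation progress value through curves on color, size and radius.
# System hue, saturation and value inputs recolor every emitter.
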NiagaraSystem /BRCosmetics/Effects/Fort_Effects/Emotes/UltraLight/NS_UltraLight_Hand
dependencies {
 usedDataInterfaces [&NiagaraDataInterfaceCurve, &NiagaraDataInterfaceColorCurve]
 usedDynamicInputs [
  &Add_Vector
  &Multiply_VectorByFloat
  &RotateVectorByQuaternion
  &Divide_Vector
  &Divide_Float
  &MakeCustomFloatFromBool
 ]
 usedModules [
  &SystemState
  &SampleSkeletalMesh
  &self:SystemSpawnScript.SystemScriptSource.SystemScriptGraph.NiagaraNodeAssignment_0.SetVariables_#1
  &self:SystemSpawnScript.SystemScriptSource.SystemScriptGraph.NiagaraNodeAssignment_1.SetVariables_#2
  &self:SystemSpawnScript.SystemScriptSource.SystemScriptGraph.NiagaraNodeAssignment_2.SetVariables_#3
  &self:SystemSpawnScript.SystemScriptSource.SystemScriptGraph.NiagaraNodeAssignment_3.SetVariables_#4
 ]
}
emitters {
 Core {
  emitterUpdateScript [
   &EmitterState(
    Inactive Response=::NewEnumerator1
    Life Cycle Mode=::NewEnumerator1
    Loop Behavior=::NewEnumerator1
    Loop Duration Mode=::NewEnumerator1
    UseLoopDelay=false
   )
   &SpawnBurst_Instantaneous(Spawn Count=1, Use Spawn Probability=false)
   /SetVariables(Emitter.Invert:Vector4f=<-1 1 1 1>) {isSetParameters=true}
  ]
  particleSpawnScript [
   &InitializeParticle(
    Color=<rgba 0.05 0 0.0083 1.5>
    Mesh Scale Mode=::NewEnumerator3
    Mesh Uniform Scale=0.07
    Sprite Size Mode=::NewEnumerator3
    Uniform Sprite Size=0
    UsePositionOffset=false
    Write Material Random=true
   )
   &CameraOffset()
  ]
  particleUpdateScript [
   &ParticleState(Kill Particles When Lifetime Has Elapsed=false)
   /SetVariables(Particles.Position=$System.HandL_Pos) {isSetParameters=true}
   /SetVariables(Particles.Position2:NiagaraPosition=$System.HandR_Pos) {isSetParameters=true}
   &DynamicMaterialParameters(
    Index 0 Param 1=&RandomRangeFloat(Minimum=0, Maximum=1)
    Module.Param0WriteEnabled=true
    Module.Param1WriteEnabled=true
    Module.Param2WriteEnabled=true
    Module.Param3WriteEnabled=true
   )
   &ScaleColor(
    Scale Alpha=&FloatFromCurve(FloatCurve=DataInterface() {propertyValues {Curve {keys [<at 0 0>,<at 0.2 0.3>,<at 0.5 1>]}}}, CurveIndex=$System.AnimProgressInput)
    Scale RGB=&VectorFromFloat(Value=&FloatFromCurve(FloatCurve=DataInterface() {propertyValues {Curve {keys [<at 0 1>,<at 0.5 1>,<at 1 10>]}}}, CurveIndex=$System.AnimProgressInput))
    ScaleA=true
    ScaleRGB=true
   )
   &ScaleMeshSize(Scale Factor=&VectorFromFloat(Value=&FloatFromCurve(FloatCurve=DataInterface() {propertyValues {Curve {keys [<at 0 1>,<at 0.5 1>,<at 1 2>]}}}, CurveIndex=$System.AnimProgressInput)))
   &ScaleColor(
    Scale Mode=::NewEnumerator1
    Scale RGBA=&MakeVector4_FromLinearColor(LinearColor=<rgba 1 0.05 0.2083 1>)
    ScaleRGBA=true
   ) {module=ScaleColor001}
   &self:ColorAdjust(Hue:NiagaraFloat=$System.Hue, Saturation:NiagaraFloat=$System.Sat, Value:NiagaraFloat=$System.Value) {module=ColorAdjust}
  ]
  rendererClasses [&NiagaraMeshRendererProperties]
  renderers [
   {rendererClass=&NiagaraMeshRendererProperties, rendererIndex=0}
   {rendererClass=&NiagaraMeshRendererProperties, rendererIndex=1}
  ]
 }
 CoreGlow {
  emitterUpdateScript [
   &EmitterState(
    Inactive Response=::NewEnumerator1
    Life Cycle Mode=::NewEnumerator1
    Loop Behavior=::NewEnumerator1
    Loop Duration Mode=::NewEnumerator1
    UseLoopDelay=false
   )
   &SpawnBurst_Instantaneous(Spawn Count=2, Use Spawn Probability=false)
   /SetVariables(Emitter.Refine:Vector4f=<10 0.3 1.3 0.0025>) {isSetParameters=true}
  ]
  particleSpawnScript [
   &InitializeParticle(Color=<rgba 400 40 40 1>, Sprite Size Mode=::NewEnumerator3, Uniform Sprite Size=13, UsePositionOffset=false, Write Material Random=true)
   &CameraOffset(Camera Offset Amount=4)
  ]
  particleUpdateScript [
   &ParticleState(Kill Particles When Lifetime Has Elapsed=false)
   /SetVariables(Particles.Position=$System.HandL_Pos) {isSetParameters=true}
   /SetVariables(Particles.Position2:NiagaraPosition=$System.HandR_Pos) {isSetParameters=true}
   &ScaleColor(
    Scale Alpha=&FloatFromCurve(FloatCurve=DataInterface() {propertyValues {Curve {keys [<at 0 0>,<at 0.3 0>,<at 0.5 1>,<at 1 1.5>]}}}, CurveIndex=$System.AnimProgressInput)
    Scale RGB=&VectorFromFloat(Value=&FloatFromCurve(FloatCurve=DataInterface() {propertyValues {Curve {keys [<at 0 1>,<at 0.3 1>,<at 0.5 1>,<at 1 2>]}}}, CurveIndex=$System.AnimProgressInput))
    ScaleA=true
    ScaleRGB=true
   )
   &self:ColorAdjust(Hue:NiagaraFloat=$System.Hue, Saturation:NiagaraFloat=$System.Sat, Value:NiagaraFloat=$System.Value) {module=ColorAdjust}
  ]
  rendererClasses [&NiagaraSpriteRendererProperties]
  renderers [
   {rendererClass=&NiagaraSpriteRendererProperties}
   {rendererClass=&NiagaraSpriteRendererProperties, rendererIndex=1}
  ]
 }
 Empty {
  emitterUpdateScript [
   &EmitterState(
    Inactive Response=::NewEnumerator1
    Life Cycle Mode=::NewEnumerator1
    Loop Behavior=::NewEnumerator1
    Loop Duration Mode=::NewEnumerator1
    UseLoopDelay=false
   )
   &SpawnBurst_Instantaneous(Use Spawn Probability=false, UseLoopCountLimit=false)
  ]
  particleSpawnScript [
   &InitializeParticle(Color Mode=::NewEnumerator0, Lifetime=5, UsePositionOffset=false, Write Material Random=true)
   /SetVariables(Particles.Color=<rgba 10000 500.0001 1608.33 1>) {isSetParameters=true}
  ]
  particleUpdateScript [
   &ParticleState(Kill Particles When Lifetime Has Elapsed=false, Let Infinitely Lived Particles Die When Emitter Deactivates=true)
   /SetVariables(Particles.Position=$System.HandL_Pos) {isSetParameters=true}
   /SetVariables(Particles.Position2:NiagaraPosition=$System.HandR_Pos) {isSetParameters=true}
   &ScaleColor(
    Scale RGB=&VectorFromFloat(Value=&FloatFromCurve(
      FloatCurve=DataInterface() {propertyValues {Curve {keys [
          {interpMode=RCIM_Cubic, leaveTangent=0.9197, tangentMode=RCTM_Break, time=0, value=0}
          {arriveTangent=2.1216, interpMode=RCIM_Cubic, leaveTangent=5.9625, tangentMode=RCTM_Break, time=0.5, value=0.7}
          <at 1 2>
         ]}}}
      CurveIndex=$System.AnimProgressInput
     ))
    ScaleA=true
    ScaleRGB=true
   )
   &ScaleColor(
    Scale RGB=&Lerp_Vector(
     A=<1 1 1>
     B=&VectorFromFloat(Value=&Waveform("Waveform [1]"=::NewEnumerator12, "[1] Frequency"=20, "[1] Amplitude Min/Max"=<0.5 1>))
     Alpha=&MakeFloatFromVector(
      Vector=&VectorFromFloat(Value=&FloatFromCurve(
        FloatCurve=DataInterface() {propertyValues {Curve {keys [
            {interpMode=RCIM_Cubic, leaveTangent=0.9197, tangentMode=RCTM_Break, time=0, value=0}
            {arriveTangent=1.1459, interpMode=RCIM_Cubic, leaveTangent=5.9625, tangentMode=RCTM_Break, time=0.5, value=0.35}
            <at 1 2>
           ]}}}
        CurveIndex=$System.AnimProgressInput
        Scale Curve=0.5
       ))
      Channel=::NewEnumerator0
     )
    )
    ScaleA=true
    ScaleRGB=true
   ) {module=ScaleColor001}
   /SetVariables(Particles.LightRadius=&MakeFloatFromVector(
     Vector=&VectorFromFloat(Value=&FloatFromCurve(
       FloatCurve=DataInterface() {propertyValues {Curve {keys [<at 0 1>, {arriveTangent=2.1216, interpMode=RCIM_Cubic, leaveTangent=0.5506, tangentMode=RCTM_Break, time=0.5, value=1}, <at 1 5>]}}}
       CurveIndex=$System.AnimProgressInput
       Scale Curve=100
      ))
     Channel=::NewEnumerator0
    )) {isSetParameters=true}
   &self:ColorAdjust(Hue:NiagaraFloat=$System.Hue, Saturation:NiagaraFloat=$System.Sat, Value:NiagaraFloat=$System.Value) {module=ColorAdjust}
  ]
  rendererClasses [&NiagaraLightRendererProperties]
  renderers [
   {rendererClass=&NiagaraLightRendererProperties, rendererIndex=0}
   {rendererClass=&NiagaraLightRendererProperties, rendererIndex=1}
  ]
 }
 EnergyCenter {
  emitterUpdateScript [
   &EmitterState(
    Inactive Response=::NewEnumerator1
    Life Cycle Mode=::NewEnumerator1
    Loop Behavior=::NewEnumerator1
    Loop Duration Mode=::NewEnumerator1
    UseLoopDelay=false
   )
   &SpawnBurst_Instantaneous(Use Spawn Probability=false, UseLoopCountLimit=false)
  ]
  particleSpawnScript [
   &InitializeParticle(
    Color=<rgba 13 13 13 1>
    Lifetime=5
    Mesh Scale Mode=::NewEnumerator3
    Mesh Uniform Scale=0.09
    UsePositionOffset=false
    Write Material Random=true
   )
   &CameraOffset()
  ]
  particleUpdateScript [
   &ParticleState(Kill Particles When Lifetime Has Elapsed=false, Let Infinitely Lived Particles Die When Emitter Deactivates=true)
   &DynamicMaterialParameters(
    Index 0 Param 1=&Multiply_Float(A=$Particles.Age, B=1.25)
    Module.Param0WriteEnabled=true
    Module.Param1WriteEnabled=true
    Module.Param2WriteEnabled=true
    Module.Param3WriteEnabled=true
   )
   /SetVariables(Particles.Position=$System.HandL_Pos) {isSetParameters=true}
   /SetVariables(Particles.Position2:NiagaraPosition=$System.HandR_Pos) {isSetParameters=true}
   &ScaleColor(
    Scale Alpha=&FloatFromCurve(FloatCurve=DataInterface() {propertyValues {Curve {keys [<at 0 0>,<at 0.5 1>,<at 1 1>]}}}, CurveIndex=$System.AnimProgressInput)
    Scale RGB=&MakeVectorFromLinearColorRGB(Color=&ColorFromCurve(ColorCurve=DataInterface() {propertyValues {
        AlphaCurve {keys [<at 0 1 RCIM_Cubic>,<at 1 1 RCIM_Cubic>]}
        BlueCurve {keys [<at 0 0.05 RCIM_Cubic>,<at 1 0.05 RCIM_Cubic>]}
        GreenCurve {keys [<at 0 0 RCIM_Cubic>,<at 1 0 RCIM_Cubic>]}
        RedCurve {keys [<at 0 1 RCIM_Cubic>,<at 1 1 RCIM_Cubic>]}
       }}))
    ScaleA=true
    ScaleRGB=true
   )
   &ScaleMeshSize(Scale Factor=&VectorFromFloat(Value=&FloatFromCurve(FloatCurve=DataInterface() {propertyValues {Curve {keys [<at 0 1>,<at 0.5 1>,<at 1 1.5>]}}}, CurveIndex=$System.AnimProgressInput)))
   &self:ColorAdjust(Hue:NiagaraFloat=$System.Hue, Saturation:NiagaraFloat=$System.Sat, Value:NiagaraFloat=$System.Value) {module=ColorAdjust}
  ]
  rendererClasses [&NiagaraMeshRendererProperties]
  renderers [
   {rendererClass=&NiagaraMeshRendererProperties, rendererIndex=0}
   {rendererClass=&NiagaraMeshRendererProperties, rendererIndex=1}
  ]
 }
 ParticleDecal_Glow {
  emitterUpdateScript [
   &EmitterState(
    Inactive Response=::NewEnumerator1
    Life Cycle Mode=::NewEnumerator1
    Loop Behavior=::NewEnumerator1
    Loop Duration Mode=::NewEnumerator1
    UseLoopDelay=false
   )
   &SpawnBurst_Instantaneous(Use Spawn Probability=false, UseLoopCountLimit=false)
  ]
  particleSpawnScript [&InitializeParticle(
    Color=<rgba 1 0.03 0.0785 1>
    Lifetime=5
    Sprite Size Mode=::NewEnumerator3
    Uniform Sprite Size=50
    UsePositionOffset=false
    Write Material Random=true
   )]
  particleUpdateScript [
   &ParticleState(Kill Particles When Lifetime Has Elapsed=false, Let Infinitely Lived Particles Die When Emitter Deactivates=true)
   /SetVariables(Particles.Position=$System.HandL_Pos) {isSetParameters=true}
   /SetVariables(Particles.Position2:NiagaraPosition=$System.HandR_Pos) {isSetParameters=true}
   &CameraOffset(Camera Offset Amount=20)
   &DynamicMaterialParameters(
    Index 0 Param 1=&FloatFromCurve(
     FloatCurve=DataInterface() {propertyValues {Curve {keys [<at 0 0>,<at 0.5 1>]}}}
     CurveIndex=$System.AnimProgressInput
     Scale Curve=&FloatFromCurve(
      FloatCurve=DataInterface() {propertyValues {Curve {keys [
          {interpMode=RCIM_Cubic, leaveTangent=2.15, tangentMode=RCTM_Break, time=0, value=0}
          {interpMode=RCIM_Cubic, tangentMode=RCTM_Break, time=0.5, value=1}
          {interpMode=RCIM_Cubic, tangentMode=RCTM_Break, time=1, value=2}
         ]}}}
      CurveIndex=$System.AnimProgressInput
     )
    )
    Index 0 Param 2=1
    Index 0 Param 3=&FloatFromCurve(
     FloatCurve=DataInterface() {propertyValues {Curve {keys [{interpMode=RCIM_Cubic, leaveTangent=2.4319, tangentMode=RCTM_Break, time=0, value=0}, <at 0.5 0.6>, <at 1 1.5>]}}}
     CurveIndex=$System.AnimProgressInput
    )
    Module.Param0WriteEnabled=true
    Module.Param1WriteEnabled=true
    Module.Param2WriteEnabled=true
    Module.Param3WriteEnabled=true
   )
   &ScaleSpriteSize(
    Uniform Curve Index=$System.AnimProgressInput
    Uniform Curve Sprite Scale=DataInterface() {propertyValues {Curve {keys [{interpMode=RCIM_Cubic, tangentMode=RCTM_Break, time=0.5, value=1}, {tangentMode=RCTM_User, time=1, value=1.6}]}}}
   )
   &self:ColorAdjust(Hue:NiagaraFloat=$System.Hue, Saturation:NiagaraFloat=$System.Sat, Value:NiagaraFloat=$System.Value) {module=ColorAdjust}
  ]
  rendererClasses [&NiagaraSpriteRendererProperties]
  renderers [
   {rendererClass=&NiagaraSpriteRendererProperties}
   {rendererClass=&NiagaraSpriteRendererProperties, rendererIndex=1}
  ]
 }
}
summary {
 systemName=NS_UltraLight_Hand
 emitters [
  {emitterName=CoreGlow, rendererClasses [&NiagaraSpriteRendererProperties]}
  {emitterName=Core, rendererClasses [&NiagaraMeshRendererProperties]}
  {emitterName=EnergyCenter, rendererClasses [&NiagaraMeshRendererProperties]}
  {emitterName=ParticleDecal_Glow, rendererClasses [&NiagaraSpriteRendererProperties]}
  {emitterName=Empty, rendererClasses [&NiagaraLightRendererProperties]}
 ]
 userVariables [
  User.bIsFrontEnd {type=NiagaraBool}
  User.ForHandSample {
   type=NiagaraDataInterfaceSkeletalMesh
   defaultValue {
    struct=NiagaraExt_VariableValue_DataInterface
    value {dataInterface=&self:NiagaraDataInterfaceSkeletalMesh_0}
   }
  }
  User.Progress {type=NiagaraFloat, defaultValue=0, description=""}
 ]
}
