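# Niagara system for the UltraLight emote that draws jet and glow effects at both hands of the character.
# Each emitter carries two hand positions from skeletal mesh sampling for its paired renderers, with meshes oriented by sampled rotation.
# Two animation progress inputs drive opacity, brightness, flare width and an accelerating material timeline.
# Shared hue, saturation and value inputs tint every layer, and bIsFrontEnd rebalances several for frontend display.
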NiagaraSystem /BRCosmetics/Effects/Fort_Effects/Emotes/UltraLight/NS_UltraLight_Jet
dependencies {
 usedDataInterfaces [&NiagaraDataInterfaceColorCurve, &NiagaraDataInterfaceCurve]
 usedDynamicInputs [
  &Multiply_VectorByFloat
  &Divide_Vector
  &RandomRangeFloat
  &Divide_Float
 ]
 usedModules [
  &SystemState
  &SampleSkeletalMesh
  &self:SystemSpawnScript.SystemScriptSource.SystemScriptGraph.NiagaraNodeAssignment_0.SetVariables_#1
  &self:SystemSpawnScript.SystemScriptSource.SystemScriptGraph.NiagaraNodeAssignment_1.SetVariables_#2
  &self:SystemSpawnScript.SystemScriptSource.SystemScriptGraph.NiagaraNodeAssignment_2.SetVariables_#3
  &self:SystemSpawnScript.SystemScriptSource.SystemScriptGraph.NiagaraNodeAssignment_3.SetVariables_#4
  &self:SystemSpawnScript.SystemScriptSource.SystemScriptGraph.NiagaraNodeAssignment_4.SetVariables_#5
  &self:SystemSpawnScript.SystemScriptSource.SystemScriptGraph.NiagaraNodeAssignment_5.SetVariables_#6
  &self:SystemSpawnScript.SystemScriptSource.SystemScriptGraph.NiagaraNodeAssignment_6.SetVariables_#7
 ]
}
emitters {
 CoreGlow001 {
  emitterUpdateScript [
   &EmitterState(
    Inactive Response=::NewEnumerator1
    Life Cycle Mode=::NewEnumerator1
    Loop Behavior=::NewEnumerator1
    Loop Duration Mode=::NewEnumerator1
    UseLoopDelay=false
   )
   &SpawnBurst_Instantaneous(Spawn Count=1, Use Spawn Probability=false)
   /SetVariables(Emitter.Refine:Vector4f=<5 2 1 0.5>, Emitter.Refine2:Vector4f=<1 0.2 0 0>) {isSetParameters=true}
  ]
  particleSpawnScript [
   &InitializeParticle(
    Color=<rgba 1 0.02 0.0527 0.4>
    Sprite Size Mode=::NewEnumerator3
    Uniform Sprite Size=100
    UsePositionOffset=false
    Write Material Random=true
   )
   &CameraOffset()
  ]
  particleUpdateScript [
   &ParticleState(Kill Particles When Lifetime Has Elapsed=false)
   /SetVariables(Particles.Position=$System.HandL_Pos) {isSetParameters=true}
   /SetVariables(Particles.Position2:NiagaraPosition=$System.HandR_Pos) {isSetParameters=true}
   &ScaleColor(
    Scale Alpha=&FloatFromCurve(
     FloatCurve=DataInterface() {propertyValues {Curve {keys [{tangentMode=RCTM_User, time=0, value=0}, {tangentMode=RCTM_User, time=1, value=1}]}}}
     CurveIndex=$System.AnimProgressInput
    )
    ScaleA=true
    ScaleRGB=true
   )
   &ScaleColor(
    Scale Alpha=&MakeCustomFloatFromBool(Bool=$User.bIsFrontEnd, True Float=0.333, False Float=1)
    Scale RGB=&VectorFromFloat(Value=&FloatFromCurve(
      FloatCurve=DataInterface() {propertyValues {Curve {keys [
          {interpMode=RCIM_Cubic, tangentMode=RCTM_Break, time=0, value=1}
          {arriveTangent=6.2277, interpMode=RCIM_Cubic, tangentMode=RCTM_Break, time=1, value=3}
         ]}}}
      CurveIndex=$System.AnimProgressBInput
     ))
    ScaleA=true
    ScaleRGB=true
   ) {module=ScaleColor002}
   &self:ColorAdjust(
    Hue:NiagaraFloat=$System.Hue
    Saturation:NiagaraFloat=$System.Sat
    Value:NiagaraFloat=&MakeCustomFloatFromBool(Bool=$User.bIsFrontEnd, True Float=3, False Float=$System.Value)
   ) {module=ColorAdjust}
  ]
  rendererClasses [&NiagaraSpriteRendererProperties]
  renderers [
   {rendererClass=&NiagaraSpriteRendererProperties}
   {rendererClass=&NiagaraSpriteRendererProperties, rendererIndex=1}
  ]
 }
 DarkGlow001 {
  emitterUpdateScript [
   &EmitterState(
    Inactive Response=::NewEnumerator1
    Life Cycle Mode=::NewEnumerator1
    Loop Behavior=::NewEnumerator1
    Loop Duration Mode=::NewEnumerator1
    UseLoopDelay=false
   )
   &SpawnBurst_Instantaneous(Spawn Count=1, Use Spawn Probability=false)
   /SetVariables(Emitter.HueShift=-0.13, Emitter.Refine:Vector4f=<3 1 3 1>) {isSetParameters=true}
  ]
  particleSpawnScript [
   &InitializeParticle(
    Color=<rgba 0.04 0.02 0.0233 0.5>
    Sprite Size Mode=::NewEnumerator3
    Uniform Sprite Size=113.965
    UsePositionOffset=false
    Write Material Random=true
   )
   &CameraOffset()
  ]
  particleUpdateScript [
   &ParticleState(Kill Particles When Lifetime Has Elapsed=false)
   /SetVariables(Particles.Position=$System.HandL_Pos) {isSetParameters=true}
   /SetVariables(Particles.Position2:NiagaraPosition=$System.HandR_Pos) {isSetParameters=true}
   &ScaleColor(
    Scale Alpha=&FloatFromCurve(
     FloatCurve=DataInterface() {propertyValues {Curve {keys [{interpMode=RCIM_Cubic, tangentMode=RCTM_Break, time=0.3, value=0}, {tangentMode=RCTM_User, time=1, value=1}]}}}
     CurveIndex=$System.AnimProgressInput
     Scale Curve=&MakeCustomFloatFromBool(Bool=$User.bIsFrontEnd, True Float=1, False Float=1)
    )
    ScaleA=true
    ScaleRGB=true
   )
   &self:ColorAdjust(
    Hue:NiagaraFloat=$System.Hue
    Saturation:NiagaraFloat=$System.Sat
    Value:NiagaraFloat=&MakeCustomFloatFromBool(Bool=$User.bIsFrontEnd, True Float=6, False Float=$System.Value)
   ) {module=ColorAdjust}
  ]
  rendererClasses [&NiagaraSpriteRendererProperties]
  renderers [
   {rendererClass=&NiagaraSpriteRendererProperties}
   {rendererClass=&NiagaraSpriteRendererProperties, rendererIndex=1}
  ]
 }
 Empty {
  emitterUpdateScript [
   &EmitterState(
    Inactive Response=::NewEnumerator1
    Life Cycle Mode=::NewEnumerator1
    Loop Behavior=::NewEnumerator1
    Loop Duration Mode=::NewEnumerator1
    UseLoopDelay=false
   )
   &SpawnBurst_Instantaneous(Use Spawn Probability=false, UseLoopCountLimit=false)
   /SetVariables(Emitter.Refine_Color:Vector4f=<1 1 1 0.5>) {isSetParameters=true, module=SetVariables_#8}
   /SetVariables(Emitter.Refine_Opacity:Vector4f=<1 1 1 0.5>) {isSetParameters=true, module=SetVariables_#9}
   /SetVariables(Emitter.SplineRibbonCurve=DataInterface() {propertyValues {
      ExposedName=SplineRibbonCurve
      bExposeCurve=true
      AlphaCurve {keys [<at 0 0>,<at 1 0>]}
      BlueCurve {keys [{arriveTangent=-0.3026, time=0, value=1}, <at 1 1 RCIM_Cubic>]}
      GreenCurve {keys [
        {interpMode=RCIM_Cubic, leaveTangent=0.3683, tangentMode=RCTM_Break, time=0, value=0}
        {arriveTangent=1.3808, interpMode=RCIM_Cubic, tangentMode=RCTM_Break, time=0.6, value=1}
        {arriveTangent=-0.0085, interpMode=RCIM_Cubic, tangentMode=RCTM_Break, time=1, value=0.5}
       ]}
      RedCurve {keys [
        <at 0 1>
        {interpMode=RCIM_Cubic, leaveTangent=-0.3211, tangentMode=RCTM_Break, time=0.5, value=1}
        {arriveTangent=-2.4683, interpMode=RCIM_Cubic, tangentMode=RCTM_Break, time=1, value=0.3}
       ]}
     }}) {isSetParameters=true, module=SetVariables_#10}
   /SetVariables(Emitter.Refraction=0.9) {isSetParameters=true}
  ]
  particleSpawnScript [&InitializeParticle(
    Color=<rgba 0.15 0 0.0125 0.3>
    Lifetime=5
    Mesh Scale=<1 1 15>
    Mesh Scale Mode=::NewEnumerator4
    Position Mode=::NewEnumerator1
    UsePositionOffset=false
    Write Material Random=true
   )]
  particleUpdateScript [
   &ParticleState(Kill Particles When Lifetime Has Elapsed=false, Let Infinitely Lived Particles Die When Emitter Deactivates=true)
   /SetVariables(Particles.Position=$System.HandL_Pos) {isSetParameters=true, module=SetVariables_#11}
   /SetVariables(Particles.Position2:NiagaraPosition=$System.HandR_Pos) {isSetParameters=true, module=SetVariables_#12}
   /SetVariables(Particles.MeshOrientation=&MultiplyQuaternion(
     Quaternion A=$System.SampleSkeletalMesh002.SampledRotation
     Quaternion B=&MakeQuatFromAxisAngle(Axis=<1 0 0>, AngleInDegrees=170)
    )) {isSetParameters=true, module=SetVariables_#13}
   /SetVariables(Particles.DPL:Vector4f=&MakeVector4(X=$System.RandomA, Y=1, Z=0, W=0)) {isSetParameters=true}
   /SetVariables(Particles.DPR:Vector4f=&MakeVector4(X=$System.RandomB, Y=1, Z=0, W=0)) {isSetParameters=true}
   &ScaleColor(
    Scale Alpha=&FloatFromCurve(
     FloatCurve=DataInterface() {propertyValues {Curve {keys [{tangentMode=RCTM_User, time=0, value=0}, {tangentMode=RCTM_User, time=1, value=1}]}}}
     CurveIndex=$System.AnimProgressInput
    )
    ScaleA=true
    ScaleRGB=true
   )
   &ScaleColor(
    Scale Alpha=&MakeCustomFloatFromBool(Bool=$User.bIsFrontEnd, True Float=0.2, False Float=1)
    ScaleA=true
    ScaleRGB=true
   ) {module=ScaleColor001}
   &self:ColorAdjust(
    Hue:NiagaraFloat=$System.Hue
    Saturation:NiagaraFloat=&MakeCustomFloatFromBool(Bool=$User.bIsFrontEnd, True Float=3, False Float=$System.Sat)
    Value:NiagaraFloat=&MakeCustomFloatFromBool(Bool=$User.bIsFrontEnd, True Float=5, False Float=$System.Value)
   ) {module=ColorAdjust}
  ]
  rendererClasses [&NiagaraMeshRendererProperties]
  renderers [
   {rendererClass=&NiagaraMeshRendererProperties, rendererIndex=0}
   {rendererClass=&NiagaraMeshRendererProperties, rendererIndex=1}
  ]
 }
 Empty001 {
  emitterUpdateScript [
   &EmitterState(
    Inactive Response=::NewEnumerator1
    Life Cycle Mode=::NewEnumerator1
    Loop Behavior=::NewEnumerator1
    Loop Duration Mode=::NewEnumerator1
    UseLoopDelay=false
   )
   &SpawnBurst_Instantaneous(Use Spawn Probability=false, UseLoopCountLimit=false)
   /SetVariables(Emitter.Refine_Color:Vector4f=<15 0 1 0.5>) {isSetParameters=true, module=SetVariables_#8}
   /SetVariables(Emitter.Refine_Opacity:Vector4f=<5 0.05 0.1 0.5>) {isSetParameters=true, module=SetVariables_#9}
   /SetVariables(Emitter.SplineRibbonCurve=DataInterface() {propertyValues {
      ExposedName=SplineRibbonCurve
      bExposeCurve=true
      AlphaCurve {keys [<at 0 0>,<at 1 0>]}
      BlueCurve {keys [
        {arriveTangent=0.2835, interpMode=RCIM_Cubic, tangentMode=RCTM_Break, time=0, value=1}
        {arriveTangent=1.9045, interpMode=RCIM_Cubic, leaveTangent=1.9045, tangentMode=RCTM_User, time=0.6, value=1.5}
        {arriveTangent=670.4279, interpMode=RCIM_Cubic, leaveTangent=113.1995, tangentMode=RCTM_Break, time=1, value=100}
       ]}
      GreenCurve {keys [
        {interpMode=RCIM_Cubic, leaveTangent=0.3683, tangentMode=RCTM_Break, time=0, value=0}
        {arriveTangent=1.3808, interpMode=RCIM_Cubic, tangentMode=RCTM_Break, time=0.6, value=1}
        {arriveTangent=-0.0085, interpMode=RCIM_Cubic, tangentMode=RCTM_Break, time=1, value=0.5}
       ]}
      RedCurve {keys [
        <at 0 1>
        {interpMode=RCIM_Cubic, leaveTangent=-0.3211, tangentMode=RCTM_Break, time=0.5, value=1}
        {arriveTangent=-2.4683, interpMode=RCIM_Cubic, tangentMode=RCTM_Break, time=1, value=0.3}
       ]}
     }}) {isSetParameters=true, module=SetVariables_#10}
  ]
  particleSpawnScript [&InitializeParticle(
    Color=<rgba 0.6 0 0.03 0.3>
    Lifetime=5
    Mesh Scale=<1 1 13>
    Mesh Scale Mode=::NewEnumerator4
    Position Mode=::NewEnumerator1
    UsePositionOffset=false
    Write Material Random=true
   )]
  particleUpdateScript [
   &ParticleState(Kill Particles When Lifetime Has Elapsed=false, Let Infinitely Lived Particles Die When Emitter Deactivates=true)
   /SetVariables(Particles.Position=$System.HandL_Pos) {isSetParameters=true, module=SetVariables_#11}
   /SetVariables(Particles.Position2:NiagaraPosition=$System.HandR_Pos) {isSetParameters=true, module=SetVariables_#12}
   /SetVariables(Particles.MeshOrientation=&MultiplyQuaternion(
     Quaternion A=$System.SampleSkeletalMesh002.SampledRotation
     Quaternion B=&MakeQuatFromAxisAngle(Axis=<1 0 0>, AngleInDegrees=170)
    )) {isSetParameters=true, module=SetVariables_#13}
   /SetVariables(Particles.DPL:Vector4f=&MakeVector4(X=$System.RandomA, Y=1, Z=0, W=0)) {isSetParameters=true}
   /SetVariables(Particles.DPR:Vector4f=&MakeVector4(X=$System.RandomB, Y=1, Z=0, W=0)) {isSetParameters=true}
   &ScaleColor(
    Scale Alpha=&FloatFromCurve(
     FloatCurve=DataInterface() {propertyValues {Curve {keys [
         {interpMode=RCIM_Cubic, leaveTangent=0.4671, tangentMode=RCTM_Break, time=0, value=0}
         {arriveTangent=3.7658, interpMode=RCIM_Cubic, leaveTangent=3.7658, tangentMode=RCTM_User, time=0.88, value=1.5}
         {arriveTangent=30.481, interpMode=RCIM_Cubic, tangentMode=RCTM_Break, time=1, value=5}
        ]}}}
     CurveIndex=$System.AnimProgressInput
    )
    ScaleA=true
    ScaleRGB=true
   )
   &ScaleColor(
    Scale Alpha=&MakeCustomFloatFromBool(Bool=$User.bIsFrontEnd, True Float=0.6, False Float=1)
    ScaleA=true
    ScaleRGB=true
   ) {module=ScaleColor001}
   &self:ColorAdjust(Hue:NiagaraFloat=$System.Hue, Saturation:NiagaraFloat=$System.Sat, Value:NiagaraFloat=$System.Value) {module=ColorAdjust}
  ]
  rendererClasses [&NiagaraMeshRendererProperties]
  renderers [
   {rendererClass=&NiagaraMeshRendererProperties, rendererIndex=0}
   {rendererClass=&NiagaraMeshRendererProperties, rendererIndex=1}
  ]
 }
 Empty002 {
  emitterUpdateScript [
   &EmitterState(
    Inactive Response=::NewEnumerator1
    Life Cycle Mode=::NewEnumerator1
    Loop Behavior=::NewEnumerator1
    Loop Duration Mode=::NewEnumerator1
    UseLoopDelay=false
   )
   &SpawnBurst_Instantaneous(Spawn Count=1, Use Spawn Probability=false)
  ]
  particleSpawnScript [
   &InitializeParticle(
    Color=<rgba 1 0 0.05 0.05>
    Sprite Size=<514.3459 413.9584>
    Sprite Size Mode=::NewEnumerator4
    UsePositionOffset=false
    Write Material Random=true
   )
   &CameraOffset(Camera Offset Amount=50)
  ]
  particleUpdateScript [
   &ParticleState(Kill Particles When Lifetime Has Elapsed=false)
   /SetVariables(Particles.Position=$System.HandL_Pos) {isSetParameters=true, module=SetVariables_#14}
   &DynamicMaterialParameters(
    Index 0 Param 1=5
    Module.Param0WriteEnabled=true
    Module.Param1WriteEnabled=true
    Module.Param2WriteEnabled=true
    Module.Param3WriteEnabled=true
   )
   /SetVariables(Particles.Position2:NiagaraPosition=$System.HandR_Pos) {isSetParameters=true, module=SetVariables_#15}
   &ScaleColor(
    Scale Alpha=&FloatFromCurve(
     FloatCurve=DataInterface() {propertyValues {Curve {keys [{tangentMode=RCTM_User, time=0, value=0}, {tangentMode=RCTM_User, time=1, value=1}]}}}
     CurveIndex=$System.AnimProgressInput
    )
    ScaleA=true
    ScaleRGB=true
   )
   &ScaleSpriteSize(
    Scale Factor=&MakeVector2D(
     X=&FloatFromCurve(FloatCurve=DataInterface() {propertyValues {Curve {keys [<at 0 0>,<at 1 1>]}}}, CurveIndex=$System.AnimProgressInput)
     Y=1
    )
    Scale Sprite Size Mode=::NewEnumerator2
   )
   &ScaleColor(
    Scale Alpha=&Waveform("Waveform [1]"=::NewEnumerator12, "[1] Phase Scale"=15, "[1] Frequency"=1, "[1] Amplitude Min/Max"=<0.5 1>)
    ScaleA=true
    ScaleRGB=true
   ) {module=ScaleColor001}
   &self:ColorAdjust(Hue:NiagaraFloat=$System.Hue, Saturation:NiagaraFloat=$System.Sat, Value:NiagaraFloat=$System.Value) {module=ColorAdjust}
  ]
  rendererClasses [&NiagaraSpriteRendererProperties]
  renderers [
   {rendererClass=&NiagaraSpriteRendererProperties}
   {rendererClass=&NiagaraSpriteRendererProperties, rendererIndex=1}
  ]
 }
 Empty003 {
  emitterUpdateScript [
   &EmitterState(
    Inactive Response=::NewEnumerator1
    Life Cycle Mode=::NewEnumerator1
    Loop Behavior=::NewEnumerator1
    Loop Duration Mode=::NewEnumerator1
    UseLoopDelay=false
   )
   &SpawnBurst_Instantaneous(Spawn Count=1, Use Spawn Probability=false)
  ]
  particleSpawnScript [
   &InitializeParticle(
    Color=<rgba 2 0.1 0.195 0.15>
    Sprite Size=<106.0944 136.6745>
    Sprite Size Mode=::NewEnumerator4
    UsePositionOffset=false
    Write Material Random=true
   )
   &CameraOffset()
  ]
  particleUpdateScript [
   &ParticleState(Kill Particles When Lifetime Has Elapsed=false)
   /SetVariables(Particles.Position=$System.HandL_Pos) {isSetParameters=true, module=SetVariables_#14}
   &DynamicMaterialParameters(
    Index 0 Param 1=5
    Module.Param0WriteEnabled=true
    Module.Param1WriteEnabled=true
    Module.Param2WriteEnabled=true
    Module.Param3WriteEnabled=true
   )
   /SetVariables(Particles.Position2:NiagaraPosition=$System.HandR_Pos) {isSetParameters=true, module=SetVariables_#15}
   &ScaleColor(
    Scale Alpha=&FloatFromCurve(
     FloatCurve=DataInterface() {propertyValues {Curve {keys [
         {interpMode=RCIM_Cubic, tangentMode=RCTM_Break, time=0.7, value=0}
         {arriveTangent=3.3333, interpMode=RCIM_Cubic, leaveTangent=3.3333, time=0.9529, value=0.279}
         {arriveTangent=8.8341, interpMode=RCIM_Cubic, tangentMode=RCTM_Break, time=1, value=1}
        ]}}}
     CurveIndex=$System.AnimProgressInput
    )
    ScaleA=true
    ScaleRGB=true
   )
   &ScaleColor(
    Scale RGB=&VectorFromFloat(Value=&FloatFromCurve(
      FloatCurve=DataInterface() {propertyValues {Curve {keys [
          {interpMode=RCIM_Cubic, tangentMode=RCTM_Break, time=0.6, value=1}
          {arriveTangent=10, interpMode=RCIM_Cubic, leaveTangent=10, time=0.9592, value=2.1424}
          {arriveTangent=52.0315, interpMode=RCIM_Cubic, tangentMode=RCTM_Break, time=1, value=5}
         ]}}}
      CurveIndex=$System.AnimProgressInput
     ))
    ScaleA=true
    ScaleRGB=true
   ) {module=ScaleColor001}
   &ScaleSpriteSize(
    Scale Factor=&MakeVector2D(
     X=&FloatFromCurve(FloatCurve=DataInterface() {propertyValues {Curve {keys [<at 0 0>,<at 1 1>]}}}, CurveIndex=$System.AnimProgressInput)
     Y=1
    )
    Scale Sprite Size Mode=::NewEnumerator2
   )
   &self:ColorAdjust(Hue:NiagaraFloat=$System.Hue, Saturation:NiagaraFloat=$System.Sat, Value:NiagaraFloat=$System.Value) {module=ColorAdjust}
  ]
  rendererClasses [&NiagaraSpriteRendererProperties]
  renderers [
   {rendererClass=&NiagaraSpriteRendererProperties}
   {rendererClass=&NiagaraSpriteRendererProperties, rendererIndex=1}
  ]
 }
 Empty004 {
  emitterUpdateScript [
   &EmitterState(
    Inactive Response=::NewEnumerator1
    Life Cycle Mode=::NewEnumerator1
    Loop Behavior=::NewEnumerator1
    Loop Duration Mode=::NewEnumerator1
    UseLoopDelay=false
   )
   &SpawnBurst_Instantaneous(Spawn Count=1, Use Spawn Probability=false)
  ]
  particleSpawnScript [&InitializeParticle(
    Color Mode=::NewEnumerator0
    Lifetime=5
    Mesh Scale=<1.55 0.3 0.3>
    Mesh Scale Mode=::NewEnumerator4
    UsePositionOffset=false
    Write Material Random=true
   )]
  particleUpdateScript [
   &ParticleState(Kill Particles When Lifetime Has Elapsed=false, Let Infinitely Lived Particles Die When Emitter Deactivates=true)
   /SetVariables(Particles.MeshOrientation=&MultiplyQuaternion(
     Quaternion A=&MultiplyQuaternion(
      Quaternion A=$System.SampleSkeletalMesh002.SampledRotation
      Quaternion B=&MakeQuatFromAxisAngle(Axis=<0 1 0>, AngleInDegrees=-90)
     )
     Quaternion B=&MakeQuatFromAxisAngle(Axis=<0 0 1>, AngleInDegrees=10)
    )) {isSetParameters=true}
   /SetVariables(Particles.Position=&Add_Vector(
     A=$System.HandL_Pos
     B=&RotateVectorByQuaternion(Vector To Rotate=<20 0 0>, Quaternion=$Particles.MeshOrientation)
    )) {isSetParameters=true}
   /SetVariables(Particles.Position2:NiagaraPosition=&Add_Vector(
     A=$System.HandR_Pos
     B=&RotateVectorByQuaternion(Vector To Rotate=<20 0 0>, Quaternion=$Particles.MeshOrientation)
    )) {isSetParameters=true}
   &Timeline(
    Play Rate=&FloatFromCurve(
     FloatCurve=DataInterface() {propertyValues {Curve {keys [
         {interpMode=RCIM_Cubic, tangentMode=RCTM_Break, time=0.4, value=0}
         {arriveTangent=5.3368, interpMode=RCIM_Cubic, tangentMode=RCTM_Break, time=1, value=1.5}
        ]}}}
     CurveIndex=$System.AnimProgressInput
    )
    Timeline End=999999
   )
   &ScaleColor(
    Scale Alpha=&FloatFromCurve(
     FloatCurve=DataInterface() {propertyValues {Curve {keys [{interpMode=RCIM_Cubic, tangentMode=RCTM_Break, time=0, value=0}, {tangentMode=RCTM_User, time=1, value=1}]}}}
     CurveIndex=$System.AnimProgressInput
    )
    Scale RGB=&MakeVectorFromLinearColorRGB(Color=&ColorFromCurve(
      ColorCurve=DataInterface() {propertyValues {
        AlphaCurve {keys [<at 0 1 RCIM_Cubic>,<at 1 1 RCIM_Cubic>]}
        BlueCurve {keys [<at 0 0.025>, {arriveTangent=0.025, interpMode=RCIM_Cubic, leaveTangent=0.025, time=0.8, value=0.025}, <at 1 0.05 RCIM_Cubic>]}
        GreenCurve {keys [<at 0 0>,<at 0.8 0 RCIM_Cubic>,<at 1 0 RCIM_Cubic>]}
        RedCurve {keys [<at 0 0.5>, {arriveTangent=0.5, interpMode=RCIM_Cubic, leaveTangent=0.5, time=0.8, value=0.5}, <at 1 1 RCIM_Cubic>]}
       }}
      CurveIndex=$System.AnimProgressInput
     ))
    ScaleA=true
    ScaleRGB=true
   )
   &DynamicMaterialParameters(
    Index 0 Param 1=&Add_Float(A=$Particles.Timeline.Time, B=&Add_Float(A=$Particles.Age, B=352.224))
    Module.Param0WriteEnabled=true
    Module.Param1WriteEnabled=true
    Module.Param2WriteEnabled=true
    Module.Param3WriteEnabled=true
   )
   &self:ColorAdjust(Hue:NiagaraFloat=$System.Hue, Saturation:NiagaraFloat=$System.Sat, Value:NiagaraFloat=$System.Value) {module=ColorAdjust}
  ]
  rendererClasses [&NiagaraMeshRendererProperties]
  renderers [
   {rendererClass=&NiagaraMeshRendererProperties, rendererIndex=0}
   {rendererClass=&NiagaraMeshRendererProperties, rendererIndex=1}
  ]
 }
}
summary {
 systemName=NS_UltraLight_Jet
 emitters [
  {emitterName=Empty, rendererClasses [&NiagaraMeshRendererProperties]}
  {emitterName=Empty001, rendererClasses [&NiagaraMeshRendererProperties]}
  {emitterName=CoreGlow001, rendererClasses [&NiagaraSpriteRendererProperties]}
  {emitterName=DarkGlow001, rendererClasses [&NiagaraSpriteRendererProperties]}
  {emitterName=Empty002, rendererClasses [&NiagaraSpriteRendererProperties]}
  {emitterName=Empty003, rendererClasses [&NiagaraSpriteRendererProperties]}
  {emitterName=Empty004, rendererClasses [&NiagaraMeshRendererProperties]}
 ]
 userVariables [
  User.bIsFrontEnd {type=NiagaraBool}
  User.ForHandSample {
   type=NiagaraDataInterfaceSkeletalMesh
   defaultValue {
    struct=NiagaraExt_VariableValue_DataInterface
    value {dataInterface=&self:NiagaraDataInterfaceSkeletalMesh_0}
   }
  }
  User.Progress {type=NiagaraFloat, defaultValue=0, description=""}
  User.ProgressB {type=NiagaraFloat, defaultValue=0, description=""}
 ]
}
